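# Post-process style AnimBP for the SmallFry player body cosmetic that copies the attached parent mesh's pose and positions pouch and kneepad bones.
# Weighted multi-bone constraints drive those bones from spine, pelvis and leg bones.
# An IsCrouching bool blends between the plain constrained pose and a duplicate chain adding pouch and kneepad offsets for crouching.
# Kneepad constraints run again after the blend, then the borrowed M_MED_Cosmos deformation Control Rig finishes the pose.
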AnimBlueprint /BRCosmetics/Characters/Player/Male/Medium/Bodies/M_MED_SmallFry/Meshes/M_MED_SmallFry_AnimBP
target_skeleton=&/M_MED_SmallFry_Skeleton
graphs [
 AnimGraph {nodes [
   n1 = "Constraint - dyn_l_pouch_front_master (PPPPP)"!(
    ComponentPose=Local To Component(LocalPose=Copy Pose From Mesh() {bUseAttachedParent=true})
    Alpha=1
    ConstraintWeights_0=0.2457
    ConstraintWeights_1:float=0.0166
    ConstraintWeights_2:float=0.0068
    ConstraintWeights_3:float=0.6604
    ConstraintWeights_4:float=0.0705
   ) -> (Pose) {
    alphaBoolBlend {blendOption=Linear}
    boneToModify {boneName=dyn_l_pouch_front_master}
    constraintSetup [
     {targetBone {boneName=spine_01}}
     {targetBone {boneName=spine_02}}
     {targetBone {boneName=spine_03}}
     {targetBone {boneName=pelvis}}
     {targetBone {boneName=thigh_twist_01_l}}
    ]
    constraintWeights [1,1,1,1,1]
   }
   n2 = "Constraint - dyn_l_pouch_mid_master (PPPPP)"!(
    ComponentPose=@n1
    Alpha=1
    ConstraintWeights_0=0.2256
    ConstraintWeights_1:float=0.0168
    ConstraintWeights_2:float=0.0066
    ConstraintWeights_3:float=0.6877
    ConstraintWeights_4:float=0.0633
   ) -> (Pose) {
    alphaBoolBlend {blendOption=Linear}
    boneToModify {boneName=dyn_l_pouch_mid_master}
    constraintSetup [
     {targetBone {boneName=spine_01}}
     {targetBone {boneName=spine_02}}
     {targetBone {boneName=spine_03}}
     {targetBone {boneName=pelvis}}
     {targetBone {boneName=thigh_twist_01_l}}
    ]
    constraintWeights [1,1,1,1,1]
   }
   n3 = "Constraint - dyn_l_pouch_back_master (PPPPP)"!(
    ComponentPose=@n2
    Alpha=1
    ConstraintWeights_0=0.2062
    ConstraintWeights_1:float=0.0151
    ConstraintWeights_2:float=0.0063
    ConstraintWeights_3:float=0.7225
    ConstraintWeights_4:float=0.0499
   ) -> (Pose) {
    alphaBoolBlend {blendOption=Linear}
    boneToModify {boneName=dyn_l_pouch_back_master}
    constraintSetup [
     {targetBone {boneName=spine_01}}
     {targetBone {boneName=spine_02}}
     {targetBone {boneName=spine_03}}
     {targetBone {boneName=pelvis}}
     {targetBone {boneName=thigh_twist_01_l}}
    ]
    constraintWeights [1,1,1,1,1]
   }
   n4 = "Constraint - dyn_r_pouch_front_master (PPPPP)"!(
    ComponentPose=@n3
    Alpha=1
    ConstraintWeights_0=0.0775
    ConstraintWeights_1:float=0.2383
    ConstraintWeights_2:float=0.0198
    ConstraintWeights_3:float=0.0079
    ConstraintWeights_4:float=0.6565
   ) -> (Pose) {
    alphaBoolBlend {blendOption=Linear}
    boneToModify {boneName=dyn_r_pouch_front_master}
    constraintSetup [
     {targetBone {boneName=thigh_twist_01_r}}
     {targetBone {boneName=spine_01}}
     {targetBone {boneName=spine_02}}
     {targetBone {boneName=spine_03}}
     {targetBone {boneName=pelvis}}
    ]
    constraintWeights [1,1,1,1,1]
   }
   n5 = "Constraint - dyn_r_pouch_mid_master (PPPPP)"!(
    ComponentPose=@n4
    Alpha=1
    ConstraintWeights_0=0.0551
    ConstraintWeights_1:float=0.2211
    ConstraintWeights_2:float=0.0176
    ConstraintWeights_3:float=0.007
    ConstraintWeights_4:float=0.6992
   ) -> (Pose) {
    alphaBoolBlend {blendOption=Linear}
    boneToModify {boneName=dyn_r_pouch_mid_master}
    constraintSetup [
     {targetBone {boneName=thigh_twist_01_r}}
     {targetBone {boneName=spine_01}}
     {targetBone {boneName=spine_02}}
     {targetBone {boneName=spine_03}}
     {targetBone {boneName=pelvis}}
    ]
    constraintWeights [1,1,1,1,1]
   }
   n6 = "Constraint - dyn_r_pouch_back_master (PPPPP)"!(
    ComponentPose=@n5
    Alpha=1
    ConstraintWeights_0=0.0414
    ConstraintWeights_1:float=0.1934
    ConstraintWeights_2:float=0.0115
    ConstraintWeights_3:float=0.0055
    ConstraintWeights_4:float=0.7482
   ) -> (Pose) {
    alphaBoolBlend {blendOption=Linear}
    boneToModify {boneName=dyn_r_pouch_back_master}
    constraintSetup [
     {targetBone {boneName=thigh_twist_01_r}}
     {targetBone {boneName=spine_01}}
     {targetBone {boneName=spine_02}}
     {targetBone {boneName=spine_03}}
     {targetBone {boneName=pelvis}}
    ]
    constraintWeights [1,1,1,1,1]
   }
   n7 = "Constraint - dyn_r_kneepad_master (PPPP)"!(
    ComponentPose=@n6
    Alpha=1
    ConstraintWeights_0=0.3981
    ConstraintWeights_1:float=0.0202
    ConstraintWeights_2:float=0.0141
    ConstraintWeights_3:float=0.5676
   ) -> (Pose) {
    alphaBoolBlend {blendOption=Linear}
    boneToModify {boneName=dyn_r_kneepad_master}
    constraintSetup [
     {targetBone {boneName=thigh_r}}
     {targetBone {boneName=calf_twist_02_r}}
     {targetBone {boneName=thigh_twist_01_r}}
     {targetBone {boneName=calf_r}}
    ]
    constraintWeights [1,1,1,1]
   }
   n8 = "Constraint - dyn_l_kneepad_master (PPPP)"!(
    ComponentPose=@n7
    Alpha=1
    ConstraintWeights_0=0.5989
    ConstraintWeights_1:float=0.0148
    ConstraintWeights_2:float=0.028
    ConstraintWeights_3:float=0.3583
   ) -> (Pose) {
    alphaBoolBlend {blendOption=Linear}
    boneToModify {boneName=dyn_l_kneepad_master}
    constraintSetup [
     {targetBone {boneName=calf_l}}
     {targetBone {boneName=thigh_twist_01_l}}
     {targetBone {boneName=calf_twist_02_l}}
     {targetBone {boneName=thigh_l}}
    ]
    constraintWeights [1,1,1,1]
   }
   n9 = Component To Local(ComponentPose=@n8)
   n10 = "Constraint - dyn_l_pouch_front_master (PPPPP)"(
    ComponentPose=Local To Component(LocalPose=Copy Pose From Mesh() {bUseAttachedParent=true})
    Alpha=1
    ConstraintWeights_0=0.2457
    ConstraintWeights_1=0.0166
    ConstraintWeights_2=0.0068
    ConstraintWeights_3=0.6604
    ConstraintWeights_4=0.0705
   ) {
    alphaBoolBlend {blendOption=Linear}
    boneToModify {boneName=dyn_l_pouch_front_master}
    constraintSetup [
     {targetBone {boneName=spine_01}}
     {targetBone {boneName=spine_02}}
     {targetBone {boneName=spine_03}}
     {targetBone {boneName=pelvis}}
     {targetBone {boneName=thigh_twist_01_l}}
    ]
    constraintWeights [1,1,1,1,1]
   }
   n11 = "Constraint - dyn_l_pouch_mid_master (PPPPP)"(
    ComponentPose=@n10
    Alpha=1
    ConstraintWeights_0=0.2256
    ConstraintWeights_1=0.0168
    ConstraintWeights_2=0.0066
    ConstraintWeights_3=0.6877
    ConstraintWeights_4=0.0633
   ) {
    alphaBoolBlend {blendOption=Linear}
    boneToModify {boneName=dyn_l_pouch_mid_master}
    constraintSetup [
     {targetBone {boneName=spine_01}}
     {targetBone {boneName=spine_02}}
     {targetBone {boneName=spine_03}}
     {targetBone {boneName=pelvis}}
     {targetBone {boneName=thigh_twist_01_l}}
    ]
    constraintWeights [1,1,1,1,1]
   }
   n12 = "Constraint - dyn_l_pouch_back_master (PPPPP)"(
    ComponentPose=@n11
    Alpha=1
    ConstraintWeights_0=0.2062
    ConstraintWeights_1=0.0151
    ConstraintWeights_2=0.0063
    ConstraintWeights_3=0.7225
    ConstraintWeights_4=0.0499
   ) {
    alphaBoolBlend {blendOption=Linear}
    boneToModify {boneName=dyn_l_pouch_back_master}
    constraintSetup [
     {targetBone {boneName=spine_01}}
     {targetBone {boneName=spine_02}}
     {targetBone {boneName=spine_03}}
     {targetBone {boneName=pelvis}}
     {targetBone {boneName=thigh_twist_01_l}}
    ]
    constraintWeights [1,1,1,1,1]
   }
   n13 = "Constraint - dyn_r_pouch_front_master (PPPPP)"(
    ComponentPose=@n12
    Alpha=1
    ConstraintWeights_0=0.0775
    ConstraintWeights_1=0.2383
    ConstraintWeights_2=0.0198
    ConstraintWeights_3=0.0079
    ConstraintWeights_4=0.6565
   ) {
    alphaBoolBlend {blendOption=Linear}
    boneToModify {boneName=dyn_r_pouch_front_master}
    constraintSetup [
     {targetBone {boneName=thigh_twist_01_r}}
     {targetBone {boneName=spine_01}}
     {targetBone {boneName=spine_02}}
     {targetBone {boneName=spine_03}}
     {targetBone {boneName=pelvis}}
    ]
    constraintWeights [1,1,1,1,1]
   }
   n14 = "Constraint - dyn_r_pouch_mid_master (PPPPP)"(
    ComponentPose=@n13
    Alpha=1
    ConstraintWeights_0=0.0551
    ConstraintWeights_1=0.2211
    ConstraintWeights_2=0.0176
    ConstraintWeights_3=0.007
    ConstraintWeights_4=0.6992
   ) {
    alphaBoolBlend {blendOption=Linear}
    boneToModify {boneName=dyn_r_pouch_mid_master}
    constraintSetup [
     {targetBone {boneName=thigh_twist_01_r}}
     {targetBone {boneName=spine_01}}
     {targetBone {boneName=spine_02}}
     {targetBone {boneName=spine_03}}
     {targetBone {boneName=pelvis}}
    ]
    constraintWeights [1,1,1,1,1]
   }
   n15 = "Constraint - dyn_r_pouch_back_master (PPPPP)"(
    ComponentPose=@n14
    Alpha=1
    ConstraintWeights_0=0.0414
    ConstraintWeights_1=0.1934
    ConstraintWeights_2=0.0115
    ConstraintWeights_3=0.0055
    ConstraintWeights_4=0.7482
   ) {
    alphaBoolBlend {blendOption=Linear}
    boneToModify {boneName=dyn_r_pouch_back_master}
    constraintSetup [
     {targetBone {boneName=thigh_twist_01_r}}
     {targetBone {boneName=spine_01}}
     {targetBone {boneName=spine_02}}
     {targetBone {boneName=spine_03}}
     {targetBone {boneName=pelvis}}
    ]
    constraintWeights [1,1,1,1,1]
   }
   n16 = "Transform (Modify) Bone - Bone: dyn_l_pouch_back_master"(ComponentPose=@n15, Alpha=1, Translation="0,-2.5,0", Rotation="0,0,15", Scale="1,1,1") {
    rotationMode=BMM_Additive
    scaleMode=BMM_Additive
    translationMode=BMM_Additive
    alphaBoolBlend {blendOption=Linear}
    boneToModify {boneName=dyn_l_pouch_back_master}
    rotation {roll=15}
    translation {y=-2.5}
   }
   n17 = "Transform (Modify) Bone - Bone: dyn_l_pouch_mid_master"(ComponentPose=@n16, Alpha=1, Translation="0,-2.5,0", Rotation="0,0,15", Scale="1,1,1") {
    rotationMode=BMM_Additive
    scaleMode=BMM_Additive
    translationMode=BMM_Additive
    alphaBoolBlend {blendOption=Linear}
    boneToModify {boneName=dyn_l_pouch_mid_master}
    rotation {roll=15}
    translation {y=-2.5}
   }
   n18 = "Transform (Modify) Bone - Bone: dyn_l_pouch_front_master"(ComponentPose=@n17, Alpha=1, Translation="0,-2.5,0", Rotation="0,0,15", Scale="1,1,1") {
    rotationMode=BMM_Additive
    scaleMode=BMM_Additive
    translationMode=BMM_Additive
    alphaBoolBlend {blendOption=Linear}
    boneToModify {boneName=dyn_l_pouch_front_master}
    rotation {roll=15}
    translation {y=-2.5}
   }
   n19 = "Transform (Modify) Bone - Bone: dyn_r_pouch_back_master"(ComponentPose=@n18, Alpha=1, Translation="0,-1,0", Rotation="0,0,8", Scale="1,1,1") {
    rotationMode=BMM_Additive
    scaleMode=BMM_Additive
    translationMode=BMM_Additive
    alphaBoolBlend {blendOption=Linear}
    boneToModify {boneName=dyn_r_pouch_back_master}
    rotation {roll=8}
    translation {y=-1}
   }
   n20 = "Transform (Modify) Bone - Bone: dyn_r_pouch_mid_master"(ComponentPose=@n19, Alpha=1, Translation="0,-1,0", Rotation="0,0,8", Scale="1,1,1") {
    rotationMode=BMM_Additive
    scaleMode=BMM_Additive
    translationMode=BMM_Additive
    alphaBoolBlend {blendOption=Linear}
    boneToModify {boneName=dyn_r_pouch_mid_master}
    rotation {roll=8}
    translation {y=-1}
   }
   n21 = "Transform (Modify) Bone - Bone: dyn_r_pouch_front_master"(ComponentPose=@n20, Alpha=1, Translation="0,-1,0", Rotation="0,0,8", Scale="1,1,1") {
    rotationMode=BMM_Additive
    scaleMode=BMM_Additive
    translationMode=BMM_Additive
    alphaBoolBlend {blendOption=Linear}
    boneToModify {boneName=dyn_r_pouch_front_master}
    rotation {roll=8}
    translation {y=-1}
   }
   n22 = "Constraint - dyn_r_kneepad_master (PPPP)"(ComponentPose=@n21, Alpha=1, ConstraintWeights_0=0.3981, ConstraintWeights_1=0.0202, ConstraintWeights_2=0.0141, ConstraintWeights_3=0.5676) {
    alphaBoolBlend {blendOption=Linear}
    boneToModify {boneName=dyn_r_kneepad_master}
    constraintSetup [
     {targetBone {boneName=thigh_r}}
     {targetBone {boneName=calf_twist_02_r}}
     {targetBone {boneName=thigh_twist_01_r}}
     {targetBone {boneName=calf_r}}
    ]
    constraintWeights [1,1,1,1]
   }
   n23 = "Constraint - dyn_l_kneepad_master (PPPP)"(ComponentPose=@n22, Alpha=1, ConstraintWeights_0=0.5989, ConstraintWeights_1=0.0148, ConstraintWeights_2=0.028, ConstraintWeights_3=0.3583) {
    alphaBoolBlend {blendOption=Linear}
    boneToModify {boneName=dyn_l_kneepad_master}
    constraintSetup [
     {targetBone {boneName=calf_l}}
     {targetBone {boneName=thigh_twist_01_l}}
     {targetBone {boneName=calf_twist_02_l}}
     {targetBone {boneName=thigh_l}}
    ]
    constraintWeights [1,1,1,1]
   }
   n24 = "Transform (Modify) Bone - Bone: dyn_l_kneepad"(ComponentPose=@n23, Alpha=1, Translation="0.2,1,-0.5", Rotation="0,0,0", Scale="1,1,1") {
    rotationMode=BMM_Additive
    translationMode=BMM_Additive
    alphaBoolBlend {blendOption=Linear}
    boneToModify {boneName=dyn_l_kneepad}
    translation=<0.2 1 -0.5>
   }
   n25 = "Transform (Modify) Bone - Bone: dyn_r_kneepad"(ComponentPose=@n24, Alpha=1, Translation="-0.8,1,-0.5", Rotation="0,0,0", Scale="1,1,1") {
    rotationMode=BMM_Additive
    translationMode=BMM_Additive
    alphaBoolBlend {blendOption=Linear}
    boneToModify {boneName=dyn_r_kneepad}
    translation=<-0.8 1 -0.5>
   }
   n26 = Component To Local(ComponentPose=@n25)
   n27 = Get bIsCrouching()
   n28 = "Constraint - dyn_r_kneepad_master (PPPP)"(
    ComponentPose=Local To Component(LocalPose=Blend Poses by bool(BlendPose_0=@n26, BlendPose_1=@n9, bActiveValue=@n27) {blendType=Linear})
    Alpha=1
    ConstraintWeights_0=0.3981
    ConstraintWeights_1=0.0202
    ConstraintWeights_2=0.0141
    ConstraintWeights_3=0.5676
   ) {
    alphaBoolBlend {blendOption=Linear}
    boneToModify {boneName=dyn_r_kneepad_master}
    constraintSetup [
     {targetBone {boneName=thigh_r}}
     {targetBone {boneName=calf_twist_02_r}}
     {targetBone {boneName=thigh_twist_01_r}}
     {targetBone {boneName=calf_r}}
    ]
    constraintWeights [1,1,1,1]
   }
   n29 = "Constraint - dyn_l_kneepad_master (PPPP)"(ComponentPose=@n28, Alpha=1, ConstraintWeights_0=0.5989, ConstraintWeights_1=0.0148, ConstraintWeights_2=0.028, ConstraintWeights_3=0.3583) {
    alphaBoolBlend {blendOption=Linear}
    boneToModify {boneName=dyn_l_kneepad_master}
    constraintSetup [
     {targetBone {boneName=calf_l}}
     {targetBone {boneName=thigh_twist_01_l}}
     {targetBone {boneName=calf_twist_02_l}}
     {targetBone {boneName=thigh_l}}
    ]
    constraintWeights [1,1,1,1]
   }
   n30 = Control Rig!(
    Source=Component To Local(ComponentPose=@n29)
    Alpha
    KneeRot_fr_l_114deg:Vector
    KneeTrans_fr_l_114deg:Vector="0.258,-3.591,0.009"
    KneeScale_fr_l_114deg:Vector="1,1,1"
    KneeScale_fr_l_130deg:Vector="1,1,1"
    KneeRot_fr_l_130deg:Vector="0,0,-51.2"
    KneeTrans_fr_l_130deg:Vector="0.258,-3.591,0.009"
   ) -> (Pose) {
    alphaBoolBlend {blendOption=Linear}
    controlRigAssetReference {blueprintRigClass=/BRCosmetics/Characters/Player/Male/Medium/Bodies/M_MED_Cosmos/Meshes/M_MED_Cosmos_Deform_CtrlRig_sculpters.M_MED_Cosmos_Deform_CtrlRig_sculpters_C}
   }
   Output Pose(Result=@n30) {layerGroup=None}
  ]}
 EventGraph {nodes [Event BlueprintUpdateAnimation()]}
]
variables []
